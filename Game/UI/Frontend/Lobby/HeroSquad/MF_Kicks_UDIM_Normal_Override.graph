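# Material function in the lobby HeroSquad UI folder that swaps texture normals for world-space vertex normals on UVs beyond the first UDIM tile in U.
# A Step on the U coordinate at 1 selects between the TextureNormals input and the vertex normals.
# An outer lerp applies that result only by the IsKicks scalar, a hologram-support parameter noted as set via ApplyOverrideMaterialWithFlag; at zero, texture normals pass through.
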MaterialFunction /Game/UI/Frontend/Lobby/HeroSquad/MF_Kicks_UDIM_Normal_Override
graph {
 properties {description=Returns VertexNormals instead of Texture Normals if UV_U is beyond 1}
 nodes [
  n1 = FunctionInput() {inputName=TextureNormals}
  n2 = VertexNormalWS()
  n3 = Step(X=ComponentMask(None=TextureCoordinate()) {r=true}) {constY=1}
  FunctionOutput(None=LinearInterpolate(
    A=@n1
    B=LinearInterpolate(A=@n1, B=@n2, Alpha=@n3)
    Alpha=ScalarParameter() {desc=Updated Via ApplyOverrideMaterialWithFlag, group=HOLOGRAMSUPPORTDONOTMODIFYMANUALLY, parameterName=IsKicks}
   ))
 ]
}
