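# MF_Holo is a hologram material function for lobby hero squad characters, outputting full material attributes.
# Fresnel, optional Z gradient, camera fade and scanlines with wiggle-driven blocks build opacity, which also scales the tinted base-texture color feeding base color and emissive.
# Switches add inner lighting, morph-target normals and vertex-color hiding, while skin masks can drive opacity mask.
# Blueprint-edge glow and wiggle vertex offset are built but left unconnected.
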MaterialFunction /Game/UI/Frontend/Lobby/HeroSquad/MF_Holo
graph {
 properties {previewBlendMode=BLEND_Masked}
 nodes [
  n1 = TextureObjectParameter() {
   parameterName=Morph Target Normals
   samplerType=SAMPLERTYPE_LinearColor
   texture=/Game/Global/Textures/Default/Blanks/T_BlackVectorDisplacement
  }
  n2 = TextureObjectParameter() {
   parameterName=Morph Target Offsets
   samplerType=SAMPLERTYPE_LinearColor
   texture=&T_BlackVectorDisplacement
  }
  n3 = MaterialFunctionCall!(
   Morph Animation=Multiply(A=ScalarParameter() {defaultValue=0.35, parameterName=Time Speed}, B=Time() {period=1})
   Morph Normal=@n1
   Morph Texture=@n2
   Number of Morph Targets=ComponentMask(None=TextureProperty(Tex=@n1)) {g=true}
   0-1 Animation Value?=StaticBool() {value=true}
   Morph Target UV
  ) {materialFunction=/Engine/Functions/Engine_MaterialFunctions02/MAXScripts/MS_VertexAnimationTools_MorphTargets}
  n4 = MaterialFunctionCall(
   Brightness=ScalarParameter() {defaultValue=1, parameterName=Fresnel Brightness, sliderMin=4}
   Exponent=ScalarParameter() {defaultValue=1.5, parameterName=Fresnel Exponent}
   Vertex / Pixel Normal=StaticSwitchParameter(True=@n3."Pixel Shader World Vertex Normal (See tooltip)", False=PixelNormalWS()) {parameterName=Use Morph Targets?}
  ) {materialFunction=&MF_Fresnel_FX}
  n5 = ScalarParameter() {defaultValue=0.8, parameterName=Opacity Bias Scale}
  n6 = ComponentMask(None=WorldPosition().XYZ) {b=true}
  n7 = ScalarParameter() {defaultValue=376, group=ZGradient, parameterName=StructureHeight}
  n8 = Add(
   A=ScalarParameter() {group=ZGradient, parameterName=ObjectPositionZ}
   B=Multiply(
    A=DynamicParameter() {defaultValue=<rgba 1 0 0 0>, paramNames [EdgeSpeed, BiasScale, Gradient Position, TriggerWiggle]}.Gradient Position
    B=@n7
   )
  )
  n9 = ScalarParameter() {defaultValue=0.5, group=ZGradient, parameterName=Gradient Thickness}
  n10 = Clamp(None=Divide(A=Subtract(A=@n6, B=@n8), B=Multiply(A=@n7, B=@n9)) {constB=384})
  n11 = Multiply(A=@n10, B=OneMinus(None=@n10))
  n12 = ScalarParameter() {defaultValue=4, group=ZGradient, parameterName=Gradient Brightness}
  n13 = Multiply(
   A=Add(A=Multiply(A=@n4.Fresnel Inverted, B=@n5), B=OneMinus(None=@n5))
   B=StaticSwitchParameter(True=Multiply(A=@n11, B=@n12) {constB=4}, False=Constant() {r=1}) {parameterName=Use Z Gradient?}
  )
  n14 = ComponentMask(None=WorldPosition().XYZ) {b=true}
  n15 = ScalarParameter() {parameterName=FresnelEdgesZ_AttenOffset}
  n16 = ComponentMask(None=MaterialFunctionCall() {materialFunction=&ObjectLocalBounds}.Local Bounds Size) {b=true}
  n17 = ScalarParameter() {parameterName=FresnelEdgesZ_AttenScale}
  n18 = Saturate(None=Divide(A=Add(A=@n14, B=@n15) {constB=10}, B=Multiply(A=@n16, B=@n17) {constB=1.2}) {constB=384})
  n19 = AppendVector(A=ScalarParameter() {defaultValue=0.0075, parameterName=Offset U}, B=ScalarParameter() {defaultValue=0.0006, parameterName=Offset V})
  n20 = TextureCoordinate()
  n21 = FunctionInput(Preview=TextureObject() {samplerType=SAMPLERTYPE_Normal, texture=/Game/Effects/Fort_Effects/Effects/Atlas/Textures/T_Atlas_N}) {bUsePreviewValueAsDefault=true, inputName=Normal, inputType=FunctionInput_Texture2D}
  n22 = MaterialFunctionCall!(TextureNormals=TextureSample(UVs=Add(A=@n19, B=@n20), Tex=@n21) {samplerType=SAMPLERTYPE_Normal}.RGB) {materialFunction=&/MF_Kicks_UDIM_Normal_Override}
  n23 = MaterialFunctionCall(TextureNormals=TextureSample(Tex=@n21) {samplerType=SAMPLERTYPE_Normal, texture=&T_Atlas_N}.RGB) {materialFunction=&MF_Kicks_UDIM_Normal_Override}
  n24 = Saturate(None=DotProduct(A=Subtract(A=@n22.Result, B=@n23.Result), B=Constant() {r=1}))
  n25 = ScalarParameter() {defaultValue=8, parameterName=Blueprint Edge Exponent}
  n26 = Multiply(A=@n18, B=StaticSwitchParameter(True=Constant(), False=Power(Base=@n24, Exp=@n25)) {parameterName=Use Morph Targets?})
  Add(A=@n13, B=Multiply(A=@n26, B=ScalarParameter() {defaultValue=2.5, parameterName=Blueprint Edge Brightness})) {constB=0, desc=Skipping since it's causing some characters to look bad in the lobby}
  Constant() {r=1}
  Constant() {r=1}
  n27 = Subtract(A=VectorParameter() {parameterName=ObjectPosition, defaultValue=<rgba 0 0 250 0>}.RGB, B=WorldPosition().XYZ)
  n28 = ConstantBiasScale(None=DotProduct(A=PixelNormalWS(), B=Normalize(VectorInput=@n27)))
  n29 = LinearInterpolate(
   A=ScalarParameter() {defaultValue=0.2, parameterName=Inner Opacity Dark}
   B=ScalarParameter() {defaultValue=1, parameterName=Inner Opacity Bright}
   Alpha=@n28
  )
  n30 = Constant() {r=1}
  n31 = Constant() {r=1}
  n32 = ScalarParameter() {defaultValue=200, parameterName=Near Camera Fade Distance}
  n33 = MaterialFunctionCall!(
   WPO Intensity A=VectorParameter() {parameterName=wiggle1, defaultValue=<rgba 0.2 0.4 0.3524 1>}.RGB
   WPO Intensity B=VectorParameter() {parameterName=wiggle1_1, defaultValue=<rgba 0 1 0 1>}.RGB
   Time
   Bias=StaticSwitchParameter(
    True=DynamicParameter() {defaultValue=<rgba 1 0 0 0>, paramNames [EdgeSpeed, BiasScale, Gradient Position, TriggerWiggle]}.TriggerWiggle
    False=Constant()
   ) {defaultValue=true, parameterName=Use Param4 for Wiggle}
  ) {materialFunction=/Game/Effects/Fort_Effects/Effects/Atlas/Materials/MF_Atlas_HologramWiggle_01}
  n34 = MaterialFunctionCall!(
   Number of Scanlines=ScalarParameter() {defaultValue=150, parameterName=Scanline Amount}
   Scanline Speed
   "Scanlines Thickness (Bias)"=ScalarParameter() {defaultValue=-0.15, parameterName=ScanlineSize}
   Scanlines Dark Value=ScalarParameter() {defaultValue=0.5, parameterName=Scanlines Dark}
   Scanlines Light Value=ScalarParameter() {defaultValue=1, parameterName=Scanlines Light}
   Block Opacity=StaticSwitchParameter(True=@n33.Intense 0-1, False=Constant()) {defaultValue=true, parameterName=Use Blocks?}
   UVs
   Time=Multiply(A=Time(), B=ScalarParameter() {defaultValue=0.1, parameterName=ScanlinesTimeMultiplier})
   Block Repeat
  ) {materialFunction=/Game/Effects/Fort_Effects/Effects/Atlas/Materials/MF_Atlas_Scanlines_01}
  n35 = Multiply(
   A=Multiply(A=@n13, B=MaterialFunctionCall(Distance=@n32) {materialFunction=&Near_Camera_Fade}.Result)
   B=@n34.Result
  )
  n36 = Add(
   A=Multiply(A=StaticSwitchParameter(True=@n29, False=@n30) {parameterName=Use Inner Light?}, B=Multiply(A=@n31, B=@n35))
   B=Multiply(A=@n34.Secondary Lines, B=ScalarParameter() {defaultValue=1, parameterName=Secondary Line Opacity})
  )
  n37 = ScalarParameter() {defaultValue=0.05, parameterName=Inner Lighting Dark}
  n38 = ScalarParameter() {defaultValue=3, parameterName=Inner Lighting Bright}
  n39 = TextureObject() {texture=/Game/Effects/Fort_Effects/Effects/Atlas/Textures/T_Atlas_D}
  n40 = StaticSwitchParameter(
   True=Constant()
   False=TextureSample(Tex=FunctionInput(Preview=@n39) {bUsePreviewValueAsDefault=true, inputName=BaseColor, inputType=FunctionInput_Texture2D}) {texture=&T_Atlas_N}.RGB
  ) {parameterName=Use Morph Targets?}
  n41 = VectorParameter() {parameterName=Glow Color Inner, defaultValue=<rgba 0.1 0.175 1 1>}
  n42 = Clamp(None=Add(A=@n40, B=ScalarParameter() {defaultValue=0.2, parameterName=Base Color Bias}))
  n43 = Multiply(A=@n18, B=@n4.Fresnel Inverted) {constA=1}
  n44 = Multiply(A=ScalarParameter() {defaultValue=0.5, parameterName=Secondary Line Brightness}, B=@n34.Secondary Lines)
  n45 = Multiply(
   A=StaticSwitchParameter(True=LinearInterpolate(A=@n37, B=@n38, Alpha=@n28), False=@n30) {parameterName=Use Inner Light?}
   B=LinearInterpolate(
    A=LinearInterpolate(A=@n40, B=@n41.RGB, Alpha=@n42)
    B=VectorParameter() {parameterName=Glow Color Outer, defaultValue=<rgba 0.55 2.125 10 1>}.RGB
    Alpha=Add(A=@n43, B=@n44)
   )
  )
  n46 = Multiply(A=Multiply(A=@n45, B=@n31), B=@n36)
  n47 = ScalarParameter() {
   defaultValue=0.9
   desc="Setting this too low will produce \"anotomically correct\" highlights - if you get what I'm saying"
   parameterName=RoughnessAmount
  }
  n48 = Multiply(
   A=VectorParameter() {parameterName=FresnelColorReplaceSpec, defaultValue=<rgba 0.0854 0.1058 0.1302 1>}.RGB
   B=Fresnel(ExponentIn=ScalarParameter() {defaultValue=1, parameterName=FresnellPoweReplacSpec}, Normal=PixelNormalWS())
  )
  n49 = TextureSampleParameter2D() {
   group=Skin FX
   parameterName=SkinFX_Mask
   samplerType=SAMPLERTYPE_Masks
   sortPriority=1
   texture=&T_WhiteMask
  }
  n50 = LinearInterpolate(B=@n49.R, Alpha=ScalarParameter() {group=HologramMasking, parameterName=UseSkinFXMaskForOpacity}) {constA=1}
  n51 = TextureSampleParameter2D() {
   group=Skin FX
   parameterName=TechArtMask
   samplerType=SAMPLERTYPE_Masks
   sortPriority=1
   texture=&T_WhiteMask
  }
  n52 = ScalarParameter() {group=HologramMasking, parameterName=UseTechArtMaskForOpacity}
  n53 = MaterialFunctionCall(TextureNormals=TextureSample(Tex=@n21) {samplerType=SAMPLERTYPE_Normal, texture=&T_Atlas_N}.RGB) {materialFunction=&MF_Kicks_UDIM_Normal_Override}
  n54 = Constant()
  n55 = MaterialFunctionCall!(WPO) {materialFunction=/Game/Packages/Fortress_SharedMaterials/Base_Material_Functions/Fortnite_HideElementsWithVertexColor}
  n56 = Constant()
  FunctionOutput(None=MakeMaterialAttributes!(
    BaseColor=@n46
    Metallic
    Specular
    Roughness=@n47
    Anisotropy
    EmissiveColor=Add(A=@n48, B=@n46)
    Opacity=@n36
    OpacityMask=LinearInterpolate(A=@n50, B=@n51.R, Alpha=@n52) {constA=1}
    Normal=StaticSwitchParameter(True=@n3."Pixel Shader World Vertex Normal (See tooltip)", False=@n53.Result) {parameterName=Use Morph Targets?}
    Tangent
    WorldPositionOffset=StaticSwitchParameter(True=@n55.Result, False=@n56) {parameterName=UseVertexMaskin}
    SubsurfaceColor
    ClearCoat
    ClearCoatRoughness
    AmbientOcclusion
    Refraction
    CustomizedUVs_0
    CustomizedUVs_1
    CustomizedUVs_2=StaticSwitchParameter(True=@n3."Custom UV 2 (Needed)", False=@n54) {parameterName=Use Morph Targets?}
    CustomizedUVs_3=StaticSwitchParameter(True=@n3."Custom UV 3 (Needed)", False=@n54) {parameterName=Use Morph Targets?}
    CustomizedUVs_4
    CustomizedUVs_5
    CustomizedUVs_6
    CustomizedUVs_7
    PixelDepthOffset
    ShadingModel
    Displacement
   ))
  Reroute(None=LinearInterpolate(A=@n45, B=VectorParameter() {parameterName=Glow Color Lines, defaultValue=<rgba 7 6.0667 1.4 1>}.RGB, Alpha=@n26) {desc=Skipping since it's causing some characters to look bad in the lobby})
  StaticSwitchParameter(True=Add(A=@n33.Result, B=@n3.World Position Offset), False=@n33.Result) {parameterName=Use Morph Targets?}
 ]
}
